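# Looping Niagara glow effect for the PistolCanary spire monolith, pinned to its skeletal mesh through one sprite-and-light emitter.
# An 11-particle burst assigns a spawn group per bone or socket with array offsets, masking one group into a much larger glow.
# System-age curves settle an initial brightness flare, while a normalized velocity value drives the rate of a counter fed to dynamic material parameters.
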NiagaraSystem /PistolCanary/Assets/FX/Monoliths/NS_PistolCanary_Spire_Glow_Loop
dependencies {
 usedDataInterfaces [&NiagaraDataInterfaceArrayFloat3, &NiagaraDataInterfaceCurve, &NiagaraDataInterfaceSkeletalMesh]
 usedDynamicInputs [&ClampFloat, &RemapRange, &VectorLength]
 usedModules [
  &self:SystemSpawnScript.SystemScriptSource.SystemScriptGraph.NiagaraNodeAssignment_3.SetVariables_#1
  &SystemState
  &InterpolateOverTime
  &self:SystemSpawnScript.SystemScriptSource.SystemScriptGraph.NiagaraNodeAssignment_0.SetVariables_#2
  &self:SystemSpawnScript.SystemScriptSource.SystemScriptGraph.NiagaraNodeAssignment_1.SetVariables_#3
 ]
}
emitters {Decals {
  emitterUpdateScript [
   &EmitterState(
    Inactive Response=::NewEnumerator1
    Life Cycle Mode=::NewEnumerator1
    Loop Behavior=::NewEnumerator1
    Loop Duration Mode=::NewEnumerator1
    UseLoopDelay=false
   )
   &SpawnBurst_Instantaneous(Spawn Count=11, Use Spawn Probability=false, UseLoopCountLimit=false)
  ]
  particleSpawnScript [
   &FN_SetSpawnGroup(Number of Spawn Groups=11)
   /SetVariables(Particles.VisibilityTag=&MaskIntBySpawnGroup(Masked Int=1, Passthrough Int=0, Spawn Group Mask=10)) {isSetParameters=true}
   &InitializeParticle(
    Color=&MaskLinearColorBySpawnGroup(Masked Linear Color=<rgba 10 0.5462 2.0669 0.025>, Passthrough Linear Color=<rgba 1.5 0.0886 0.3291 1>, Spawn Group Mask=10)
    Lifetime=&MaskFloatBySpawnGroup(Masked Float=5, Passthrough Float=1, Spawn Group Mask=10)
    Sprite Size=&MaskVector2dBySpawnGroup(
     Masked Vector2D=&Multiply_Vector2D_ByFloat(Float=3000, Vector2D=<1 1>)
     Passthrough Vector2D=<30 30>
     Spawn Group Mask=10
    )
    Sprite Size Mode=::NewEnumerator4
    UsePositionOffset=false
    Write Material Random=true
   )
   &CameraOffset(Camera Offset Amount=&MaskFloatBySpawnGroup(Masked Float=1000, Passthrough Float=1002, Spawn Group Mask=10))
   /SetVariables(Particles.LightGlowDynParameter:Vector4f=&MakeVector4(X=2, Y=1, Z=1, W=1)) {isSetParameters=true}
  ]
  particleUpdateScript [
   &ParticleState(Kill Particles When Lifetime Has Elapsed=false, Let Infinitely Lived Particles Die When Emitter Deactivates=true)
   &SkeletalMeshLocation(
    Bone / Socket Index=&SpawnGroup()
    Bone Sampling Mode=::NewEnumerator4
    MaskSpawnGroup=false
    Orientation Sampling=::NewEnumerator1
    Sampled Position Offset=&SelectVectorFromArray(
     Direct Array Index=&SpawnGroup()
     Vector Selection Array=DataInterface() {propertyValues {FloatData [<5 0 0>, <-5 0 0>, <0 0 0>, <10 0 0>, <-10 0 0>, <0 0 0>, <0 0 0>, <0 0 0>, <5 0 0>, <-5 0 0>, <0 0 0>]}}
    )
    Skeletal Mesh=$System.Skeletal Mesh
   )
   /Niagara/Modules/Update/Utility/IncrementOverTime(Rate of Change:NiagaraFloat=&Lerp_Float(A=-1.1, B=2, Alpha=$System.NormalizedVelocityRange))
   &DynamicMaterialParameters(
    Index 0 Param 1=&FloatFromCurve(
     FloatCurve=DataInterface() {propertyValues {Curve {keys [{arriveTangent=3.2804, interpMode=RCIM_Cubic, leaveTangent=3.2804, tangentMode=RCTM_User, time=0, value=1}, <at 0.7 2>]}}}
     CurveIndex=$System.Age
    )
    Index 0 Param 2=1
    Index 0 Param 3=&Add_Float(
     A=$Particles.IncrementOverTime.Counter
     B=&FloatFromCurve(
      FloatCurve=DataInterface() {propertyValues {Curve {keys [{arriveTangent=-2.1645, interpMode=RCIM_Cubic, leaveTangent=-2.1645, tangentMode=RCTM_User, time=0, value=1}, <at 1.25 0>]}}}
      CurveIndex=$System.Age
      Scale Curve=-0.5
     )
    )
    Index 0 Param 4=&RandomRangeFloat(Minimum=0.75, Maximum=0.9)
    Index 1 Param 1=50
    Index 1 Param 2=10
    Module.Param0WriteEnabled=true
    Module.Param0WriteEnabled1=true
    Module.Param1WriteEnabled=true
    Module.Param1WriteEnabled1=true
    Module.Param2WriteEnabled=true
    Module.Param2WriteEnabled1=false
    Module.Param3WriteEnabled=true
    Module.Param3WriteEnabled1=false
    Write Parameter Index 1=true
   )
   &ScaleColor(
    Scale RGB=&VectorFromFloat(Value=&FloatFromCurve(
      FloatCurve=DataInterface() {propertyValues {Curve {keys [{arriveTangent=-22.3603, interpMode=RCIM_Cubic, leaveTangent=-22.3603, tangentMode=RCTM_User, time=0, value=9}, <at 0.7 1>]}}}
      CurveIndex=$System.Age
     ))
    ScaleA=true
    ScaleRGB=true
   )
   /SetVariables(Particles.LightGlowColor:LinearColor=&MultiplyLinearColor(
     Linear Color=$Particles.Initial.Color
     Scale Factor=&MakeFloat4FromVectorAndFloat(
      "Vector (RGB)"=&VectorFromFloat(Value=&FloatFromCurve(
        FloatCurve=DataInterface() {propertyValues {Curve {keys [{arriveTangent=-18.9054, interpMode=RCIM_Cubic, time=0, value=12}, <at 0.7 1>]}}}
        CurveIndex=$System.Age
       ))
      "Float (Alpha)"=1
     )
    )) {isSetParameters=true}
   /SetVariables(Particles.LightGlowAlignment:Vector3f=<0 0 1>) {isSetParameters=true}
  ]
  rendererClasses [&NiagaraSpriteRendererProperties, &NiagaraLightRendererProperties]
  renderers [
   {rendererClass=&NiagaraSpriteRendererProperties}
   {rendererClass=&NiagaraSpriteRendererProperties, rendererIndex=1}
   {rendererClass=&NiagaraLightRendererProperties, rendererIndex=2}
  ]
 }}
summary {
 systemName=NS_PistolCanary_Spire_Glow_Loop
 emitters [{emitterName=Decals, rendererClasses [&NiagaraSpriteRendererProperties, &NiagaraLightRendererProperties]}]
 userVariables [
  User.GameTime {type=NiagaraFloat, defaultValue=0, description=""}
  User.RibbonIDOffset {type=NiagaraInt32}
 ]
}
